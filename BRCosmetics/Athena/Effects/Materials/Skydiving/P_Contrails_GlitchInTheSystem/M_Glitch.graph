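# Additive unlit particle material for the Glitch In The System skydiving contrail, tinting particles between two HDR colors by a sine of world height.
# Its mask draws screen-space dots weighted by an inverted scene-normal-to-camera dot product, with optional scan lines from a panning cracks texture.
# A radial UV mask and inverted depth fade confine the pattern to where sprites near scene geometry.
# The final mask switch outputs zero on mobile.
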Material /BRCosmetics/Athena/Effects/Materials/Skydiving/P_Contrails_GlitchInTheSystem/M_Glitch
graph {
 properties {
  bUsedWithNiagaraSprites=true
  bUsedWithParticleSprites=true
  blendMode=BLEND_Additive
  pixelDepthOffsetMode=PDOM_Legacy
  shadingModel=MSM_Unlit
  translucencyPass=MTP_BeforeDOF
  usedShadingModels=Unlit
 }
 nodes [
  Constant3Vector() {constant=<rgba 17.9786 30 0 0>}
  Constant3Vector() {constant=<rgba 0.0665 30 0 0>}
  Constant() {r=1}
  Constant() {r=1}
  n1 = ParticlePositionWS()
  n2 = Sine(None=Divide(A=ComponentMask(None=@n1) {b=true}) {constB=3000})
  MaterialFunctionCall(Emissive=LinearInterpolate(
    A=VectorParameter() {parameterName=EmissiveColor1, defaultValue=<rgba 30 1 0.5 1>}.RGB
    B=Constant3Vector() {constant=<rgba 60 0 0 0>}
    Alpha=Abs(None=@n2)
   )) {materialFunction=&MF_FX_EmissiveAdjustments}
  n3 = Divide(A=ScreenPosition().PixelPosition) {constB=5}
  n4 = SceneTexture() {sceneTextureId=PPI_WorldNormal}
  n5 = Constant3Vector() {constant=<rgba 0 0 1 0>}
  n6 = DotProduct(
   A=ComponentMask(None=ShadingPathSwitch(Default=@n4.Color, Deferred=@n4.Color, Forward=@n5, Mobile=@n4.Color)) {b=true, g=true, r=true}
   B=CameraVectorWS()
  )
  n7 = Multiply(
   A=SphereMask(A=Frac(None=@n3), B=Constant() {r=0.5}, Radius=Constant() {r=0.25}, Hardness=Constant())
   B=Multiply(A=OneMinus(None=@n6)) {constB=2}
  )
  n8 = Constant()
  n9 = Multiply(A=ComponentMask(None=ScreenPosition().ViewportUV) {g=true}) {constB=5}
  n10 = TextureSample(UVs=Panner(Coordinate=AppendVector(A=@n8, B=@n9)) {speedX=0.2569, speedY=0.6512}) {samplerType=SAMPLERTYPE_LinearColor, texture=/Game/Global/Textures/Default/Masks_and_Gradients/T_Cracks}
  n11 = Multiply(A=@n10.R, B=@n10.R)
  n12 = StaticSwitchParameter(True=Add(A=@n7, B=Multiply(A=@n11, B=@n11)), False=@n7) {parameterName=UseScanLine}
  n13 = Constant() {r=0.5}
  n14 = SphereMask(A=TextureCoordinate(), B=@n13, Radius=@n13, Hardness=Constant())
  n15 = Constant() {r=1}
  n16 = Constant() {r=100}
  n17 = Multiply(A=@n14, B=OneMinus(None=DepthFade(Opacity=@n15, FadeDistance=@n16)))
  n18 = Constant() {r=0.5}
  n19 = ShadingPathSwitch(Default=@n17, Mobile=SphereMask(A=TextureCoordinate(), B=@n18, Radius=@n18, Hardness=Constant() {r=-2})) {desc=Migrated from FeatureLevelSwitch}
  ShadingPathSwitch(Default=Add(A=Multiply(A=@n12, B=@n19), B=ScalarParameter() {parameterName=AlphaBias}) {constB=0.01}, Mobile=Constant()) {desc=Migrated from FeatureLevelSwitch}
  StaticSwitchParameter(True=@n14, False=@n17) {parameterName=DisableDF}
 ]
}
