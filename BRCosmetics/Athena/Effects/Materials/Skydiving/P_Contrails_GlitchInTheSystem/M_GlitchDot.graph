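# Unlit particle material for the dots in the GlitchInTheSystem skydiving contrail, usable on Cascade and Niagara sprites.
# A sine-of-UV mask blends base to emissive color for a center hotspot, scaled by particle alpha through MF_FX_EmissiveAdjustments.
# Per-particle randomized, panning digital noise builds a glitch vertex offset, plus a term collapsing sprites toward their centers near the camera.
# The mobile shading path replaces this offset with a constant zero.
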Material /BRCosmetics/Athena/Effects/Materials/Skydiving/P_Contrails_GlitchInTheSystem/M_GlitchDot
graph {
 properties {
  bUsedWithNiagaraSprites=true
  bUsedWithParticleSprites=true
  pixelDepthOffsetMode=PDOM_Legacy
  shadingModel=MSM_Unlit
  translucencyPass=MTP_BeforeDOF
  usedShadingModels=Unlit
 }
 nodes [
  Constant3Vector() {constant=<rgba 0.0058 0.2 0.1265 0>}
  Constant3Vector() {constant=<rgba 0.3812 1 0 0>}
  n1 = VectorParameter() {parameterName=Base_color, defaultValue=<rgba 0.1 0 0 1>}
  n2 = VectorParameter() {parameterName=Emissive, defaultValue=<rgba 2 0 0 1>}
  n3 = Sine(None=TextureCoordinate() {uTiling=0.5, vTiling=0.5})
  n4 = Multiply(A=ComponentMask(None=@n3) {r=true}, B=ComponentMask(None=@n3) {g=true})
  MaterialFunctionCall(Emissive=Multiply(A=LinearInterpolate(A=@n1.RGB, B=@n2.RGB, Alpha=@n4), B=ParticleColor().A)) {materialFunction=&MF_FX_EmissiveAdjustments}
  n5 = Add(A=TextureCoordinate() {uTiling=0.01, vTiling=0.01}, B=ParticleRandom())
  n6 = Multiply(A=Time(), B=ScalarParameter() {defaultValue=8, parameterName=Speeed})
  n7 = Transform(None=TextureSample(UVs=Panner(Coordinate=@n5, Time=@n6) {speedX=0.1, speedY=0.1}) {samplerType=SAMPLERTYPE_LinearColor, texture=/Game/Global/Textures/Default/Noise/T_Atlas_DigitalNoise_01}.RGB)
  n8 = ScalarParameter() {defaultValue=2, parameterName=GlitchAmount}
  n9 = ParticlePositionWS()
  n10 = Subtract(A=@n9, B=WorldPosition().XYZ)
  n11 = Saturate(None=SphereMask(A=@n9, B=CameraPositionWS(), Radius=Constant() {r=200}, Hardness=Constant() {r=0.1}))
  ShadingPathSwitch(Default=Add(A=Multiply(A=@n7, B=@n8), B=Multiply(A=@n10, B=@n11) {constB=0.5}), Mobile=Constant()) {desc=Migrated from FeatureLevelSwitch}
 ]
}
