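# Unlit translucent, two-sided trail material for the GlitchInTheSystem skydiving contrail, usable on beam trails and Niagara ribbons or sprites.
# Opacity adds a screen-space dot grid, sized by local-space trail UV, to panning crack texture, shaped by UV falloffs and optional particle alpha.
# Emissive blends two HDR colors by particle world height, scaled by time of day.
# A digital-noise UV-distortion branch appears to feed nothing.
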Material /BRCosmetics/Athena/Effects/Materials/Skydiving/P_Contrails_GlitchInTheSystem/M_GlitchTrail
graph {
 properties {
  bUsedWithBeamTrails=true
  bUsedWithNiagaraRibbons=true
  bUsedWithNiagaraSprites=true
  blendMode=BLEND_Translucent
  opacityMaskClipValue=0.95
  pixelDepthOffsetMode=PDOM_Legacy
  shadingModel=MSM_Unlit
  twoSided=true
  usedShadingModels=Unlit
 }
 nodes [
  n1 = Divide(A=ScreenPosition().PixelPosition) {constB=8}
  n2 = Constant() {r=0.5}
  n3 = ComponentMask(None=MaterialFunctionCall!(MaxiumURange=Constant() {r=750}, RangeRampHardness, MirrorTexCoordU) {materialFunction=/BRCosmetics/Weapons/FORT_Melee/Pickaxe_Venus/MF_LocalSpaceTrailUV}."LocalSpaceCoordinate(UV)") {r=true}
  n4 = Subtract(A=@n3) {constB=0.3}
  n5 = Multiply(A=SphereMask(A=Frac(None=@n1), B=@n2, Radius=Multiply(A=@n2, B=@n4), Hardness=Constant()), B=@n3)
  n6 = Constant()
  n7 = Multiply(A=ComponentMask(None=ScreenPosition().ViewportUV) {g=true}) {constB=5}
  n8 = TextureSample(UVs=Panner(Coordinate=AppendVector(A=@n6, B=@n7)) {speedX=0.2569, speedY=0.6512}) {samplerType=SAMPLERTYPE_LinearColor, texture=/Game/Global/Textures/Default/Masks_and_Gradients/T_Cracks}
  n9 = Multiply(A=Multiply(A=@n8.R, B=Subtract(A=@n3) {constB=0.6}), B=Multiply(A=@n4, B=@n4))
  n10 = Sine(None=ComponentMask(None=TextureCoordinate() {vTiling=0.5}) {g=true})
  n11 = TextureCoordinate()
  n12 = ComponentMask(None=Multiply(A=OneMinus(None=@n11)) {constB=15}) {r=true}
  n13 = Subtract(A=Multiply(A=Add(A=@n5, B=@n9), B=@n10), B=Clamp(None=Subtract(A=@n12) {constB=14}))
  LinearInterpolate(A=@n13, B=Multiply(A=@n13, B=ParticleColor().A), Alpha=ScalarParameter() {parameterName=Multiply by Particle Alpha})
  n14 = VectorParameter() {parameterName=ColorA, defaultValue=<rgba 30 1 0.5 1>}
  n15 = VectorParameter() {parameterName=ColorB, defaultValue=<rgba 60 0 0 1>}
  n16 = ComponentMask(None=ParticlePositionWS()) {b=true}
  n17 = Abs(None=Sine(None=Divide(A=@n16) {constB=3000}))
  n18 = Constant() {r=5}
  n19 = Constant() {r=1}
  MaterialFunctionCall(Emissive=Multiply(
    A=LinearInterpolate(A=@n14.RGB, B=@n15.RGB, Alpha=@n17)
    B=MaterialFunctionCall(Color_Morning=@n18, Color_Day=@n18, Color_Evening=@n18, Color_Night=@n19) {materialFunction=&Color_Time_of_Day}.Result
   )) {materialFunction=&MF_FX_EmissiveAdjustments}
  n20 = Multiply(A=@n8.R, B=@n8.R)
  Multiply(A=@n20, B=@n20)
  n21 = Sine(None=Multiply(A=ScreenPosition().ViewportUV) {constB=0.5})
  Multiply(A=Multiply(A=ComponentMask(None=@n21) {r=true}, B=ComponentMask(None=@n21) {g=true}), B=@n10)
  n22 = Constant2Vector() {g=1, r=0.1}
  n23 = ScreenPosition()
  n24 = Panner(Coordinate=@n23.ViewportUV, Time=Multiply(A=Time(), B=ScalarParameter() {defaultValue=4, parameterName=Speeed})) {speedX=0.1, speedY=0.1}
  n25 = Multiply(
   A=TextureSample(UVs=Multiply(A=@n22, B=@n24)) {samplerType=SAMPLERTYPE_LinearColor, texture=/Game/Global/Textures/Default/Noise/T_Atlas_DigitalNoise_01}.R
   B=ScalarParameter() {defaultValue=2, parameterName=GlitchAmount}
  )
  n26 = Multiply(A=ComponentMask(None=Add(A=@n23.ViewportUV, B=@n25)) {g=true}) {constB=2}
  n27 = Constant()
  TextureSample(UVs=Panner(Coordinate=AppendVector(A=@n26, B=@n27)) {speedX=0.2569, speedY=0.6512}) {samplerType=SAMPLERTYPE_LinearColor, texture=&T_Cracks}
 ]
}
