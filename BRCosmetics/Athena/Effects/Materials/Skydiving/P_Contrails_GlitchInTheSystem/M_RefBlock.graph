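# Additive unlit sprite material for cascade and Niagara particles in the Glitch In The System skydiving contrail, set to index-of-refraction mode.
# A digital noise texture, panned over tiled screen-space UVs, is recentred and normalized into a vector that suggests blocky refraction distortion.
# A separate lerp blends two constant colors by particle size.
# Particle color, random and camera-vector nodes also appear, but no links to material outputs are shown.
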Material /BRCosmetics/Athena/Effects/Materials/Skydiving/P_Contrails_GlitchInTheSystem/M_RefBlock
graph {
 properties {
  bUsedWithNiagaraSprites=true
  bUsedWithParticleSprites=true
  blendMode=BLEND_Additive
  pixelDepthOffsetMode=PDOM_Legacy
  refractionMethod=RM_IndexOfRefraction
  shadingModel=MSM_Unlit
  usedShadingModels=Unlit
 }
 nodes [
  Constant() {r=0.5}
  Constant()
  Constant() {r=1.03}
  LinearInterpolate(A=Constant3Vector() {constant=<rgba 1.3209 2 0 0>}, B=Constant3Vector(), Alpha=ComponentMask(None=ParticleSize()) {r=true})
  n1 = Panner(Coordinate=Multiply(A=ScreenPosition().ViewportUV) {constB=3}) {speedY=1}
  Normalize(VectorInput=Subtract(A=TextureSample(UVs=@n1) {samplerType=SAMPLERTYPE_LinearColor, texture=/Game/Global/Textures/Default/Noise/T_Atlas_DigitalNoise_01}.RGB) {constB=0.5})
  ParticleColor()
  ParticleColor()
  ParticleRandom()
  TextureCoordinate()
  Transform(None=CameraVectorWS()) {transformSourceType=TRANSFORMSOURCE_Camera, transformType=TRANSFORM_Tangent}
 ]
}
